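# Idle Niagara effect for the PrairieGizmo back bling, with spinning fan sprites and an orange glow attached through a socket reader.
# Fans partitions two sprites onto sockets, counter-rotates them, aligns them to socket rotation, and drives alpha by a system timeline.
# Glow writes separate front and top/bottom sprite attributes for its two renderers, plus light attributes.
# Time of day adjusts tint and brightness, while front-end modifiers tune size and intensity.
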NiagaraSystem /BRCosmetics/Accessories/FORT_Backpacks/Backpack_F_MED_PrairieGizmo/FX/NS_Backpack_PrairieGizmo_Idle
dependencies {
 usedDataInterfaces [
  &NiagaraDataInterfaceArrayFloat
  &NiagaraDataInterfaceSocketReader
  &NiagaraDataInterfaceCurve
  &NiagaraDataInterfaceColorCurve
 ]
 usedDynamicInputs [&Boolean_NotOperation, &ColorFromCurve]
 usedModules [
  &self:SystemSpawnScript.SystemScriptSource.SystemScriptGraph.NiagaraNodeAssignment_0.SetVariables_#1
  &SystemState
  &Timeline
  &self:SystemSpawnScript.SystemScriptSource.SystemScriptGraph.NiagaraNodeAssignment_1.SetVariables_#2
 ]
}
emitters {
 Fans {
  emitterUpdateScript [
   &EmitterState(
    Inactive Response=::NewEnumerator1
    Life Cycle Mode=::NewEnumerator1
    Loop Behavior=::NewEnumerator1
    Loop Duration Mode=::NewEnumerator1
    UseLoopDelay=false
   )
   &SpawnBurst_Instantaneous(Spawn Count=2, Use Spawn Probability=false, UseLoopCountLimit=false)
  ]
  particleSpawnScript [
   &PartitionParticles()
   &InitializeParticle(
    Color=<rgba 0.1911 0.17 0.2448 1>
    Lifetime=5
    Sprite Rotation Angle=&SelectFloatFromArray(Direct Array Index=$Particles.PartitionParticles.Partition, Float Selection Array=DataInterface() {propertyValues {FloatData [10,-10]}})
    Sprite Rotation Mode=::NewEnumerator0
    Sprite Size Mode=::NewEnumerator3
    Uniform Sprite Size=&SelectFloatFromArray(Direct Array Index=$Particles.PartitionParticles.Partition, Float Selection Array=DataInterface() {propertyValues {FloatData [12,-12]}})
    UsePositionOffset=false
    Write Material Random=true
   )
  ]
  particleUpdateScript [
   &ParticleState(Kill Particles When Lifetime Has Elapsed=false, Let Infinitely Lived Particles Die When Emitter Deactivates=true)
   &SpriteRotationRate(Rotation Rate=&Multiply_Float(
     A=&SelectFloatFromArray(Direct Array Index=$Particles.PartitionParticles.Partition, Float Selection Array=DataInterface() {propertyValues {FloatData [-1,1]}})
     B=200
    ))
   &SocketLocation(
    SamplePosOffset=true
    Sampled Orientation Coordinate Space:ENiagaraCoordinateSpace=::Local
    Sampled Position Offset=<-1.2 0 0>
    Socket Index=$Particles.PartitionParticles.Partition
    Socket Reader=$System.Socket Reader
    Socket Sampling Mode=::NewEnumerator3
    VelocitySpeedLimit=false
   )
   &SpriteFacingAndAlignment(
    Facing Coordinate Space=::Local
    Sprite Alignment=&RotateVectorByQuaternion(Vector To Rotate=<0 1 0>, Quaternion=$Output.SocketLocation.SampledRotation)
    Sprite Facing=&RotateVectorByQuaternion(Vector To Rotate=<1 0 0>, Quaternion=$Output.SocketLocation.SampledRotation)
    WriteAlignment=true
    WriteFacing=true
   )
   &ScaleColor(
    Scale Alpha=&Multiply_Float(A=$System.Timeline.Float, B=&FN_FrontEndFloatModifier(Gameplay=1, Front End=0.5))
    Scale RGB=&Lerp_Vector(A=<1 1 1>, B=&MakeVectorFromLinearColorRGB(Color=$System.TODColor), Alpha=0.8)
    ScaleA=true
    ScaleRGB=true
   )
  ]
  rendererClasses [&NiagaraSpriteRendererProperties]
  renderers [{rendererClass=&NiagaraSpriteRendererProperties}]
 }
 Glow {
  emitterUpdateScript [
   &EmitterState(
    Inactive Response=::NewEnumerator1
    Life Cycle Mode=::NewEnumerator1
    Loop Behavior=::NewEnumerator1
    Loop Duration Mode=::NewEnumerator1
    UseLoopDelay=false
   )
   &SpawnBurst_Instantaneous(Use Spawn Probability=false, UseLoopCountLimit=true)
  ]
  particleSpawnScript [
   &InitializeParticle(
    Color=<rgba 3 0.8311 0.171 1>
    Lifetime=5
    Mesh Scale Mode=::NewEnumerator3
    Mesh Uniform Scale=4
    Sprite Size=<25 10>
    Sprite Size Mode=::NewEnumerator4
    UsePositionOffset=false
    Write Material Random=true
   )
   &DynamicMaterialParameters(
    Index 0 Param 1=5
    Index 0 Param 2=6.0074
    Index 0 Param 4=30
    Module.Index0Param0Write=true
    Module.Index0Param1Write=true
    Module.Index0Param2Write=true
    Module.Index0Param3Write=true
   ) {module=DynamicMaterialParameters001}
   &CameraOffset(Camera Offset Amount=-20)
   &SpriteFacingAndAlignment(WriteAlignment=true, WriteFacing=true)
   /SetVariables(Particles.AlignFront:Vector3f=<1 0 0>, Particles.FacingFront:Vector3f=<0 1 -0.35>) {isSetParameters=true}
   /SetVariables(Particles.AlignTopBot:Vector3f=<1 0 1>, Particles.FacingTopBot:Vector3f=<0 0 1>) {isSetParameters=true}
   /SetVariables(
    Particles.CameraOffsetNul=1
    Particles.Glow Color=<rgba 2 0.5541 0.114 1>
    Particles.SizeFront:Vector2f=&FN_FrontEndVector2Modifier(Front End:Vector2f=<20 15>, Gameplay:Vector2f=<22 15>)
    Particles.SizeTopBot:Vector2f=<4 13>
   ) {isSetParameters=true}
   /SetVariables(
    Particles.DynamicFront:Vector4f=&MakeVector4(X=&FN_FrontEndFloatModifier(Gameplay=4, Front End=3), Y=0, Z=3, W=23)
    Particles.DynamicTop:Vector4f=<3 1 3 23>
   ) {isSetParameters=true}
  ]
  particleUpdateScript [
   &ParticleState(Kill Particles When Lifetime Has Elapsed=false)
   &SocketLocation(
    SamplePosOffset=true
    Sampled Position Coordinate Space:ENiagaraCoordinateSpace=::Local
    Sampled Position Offset=<0 -13 5.5>
    Socket Reader=$System.Socket Reader
    Socket Sampling Mode=::NewEnumerator2
    VelocitySpeedLimit=false
   )
   &FN_Light_Attributes(
    Color=<rgba 1 0.25 0 1>
    Create Color Attribute=true
    Create Exponent Attribute=true
    Create Position Attribute=true
    Create Radius Attribute=true
    Exponent=12
    Position:NiagaraPosition=$Output.SocketLocation.SampledPosition
    Radius=20
    Scale Color=&Multiply_Float(A=5, B=$System.TOD)
   )
   &ScaleColor(
    Scale Alpha=&FNTimeOfDayScalar(Night Scale=0.25, Morning Scale=1, Day Scale=0.5, Evening Scale=0.4)
    ScaleA=true
    ScaleRGB=true
   )
  ]
  rendererClasses [&NiagaraSpriteRendererProperties]
  renderers [
   {rendererClass=&NiagaraSpriteRendererProperties}
   {rendererClass=&NiagaraSpriteRendererProperties, rendererIndex=1}
  ]
 }
}
summary {
 systemName=NS_Backpack_PrairieGizmo_Idle
 emitters [
  {emitterName=Fans, rendererClasses [&NiagaraSpriteRendererProperties]}
  {emitterName=Glow, rendererClasses [&NiagaraSpriteRendererProperties]}
 ]
 userVariables [
  User.bIsFrontEnd {type=NiagaraBool}
  User.bIsFrontEndPreview {type=NiagaraBool}
  User.ReactivityEnabled {type=NiagaraBool}
 ]
}
