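# Animation Blueprint for the PrairieGizmo backpack cosmetic, animating its twin fans and motor and lowering the pack outside skydiving.
# Each update drives additive bone transforms with randomized sine motor jitter and opposite fan spins from FanRotation.
# Fans spin only with the Stage1 variant, gaining speed during skydive, parachute and elimination reactions.
# Enemy eliminations trigger three-second reactions, replayed on a looping timer in the frontend.
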AnimBlueprint /BRCosmetics/Accessories/FORT_Backpacks/Backpack_F_MED_PrairieGizmo/Meshes/F_MED_PrairieGizmo_Pack_AnimBP
target_skeleton=&/F_MED_PrairieGizmo_Pack_Skeleton
graphs [
 AnimGraph {nodes [
   n1 = "Transform (Modify) Bone - Bone: pack"!(
    ComponentPose=Local To Component(LocalPose=Input Pose())
    bAlphaBoolEnabled:bool=Get Backpack offset() -> (Backpack offset:bool)
    Translation="0,0,-3"
    Rotation="0,0,0"
    Scale
   ) -> (Pose) {
    alphaInputType=Bool
    rotationSpace=BCS_BoneSpace
    translationMode=BMM_Additive
    translationSpace=BCS_BoneSpace
    alphaBoolBlend {blendInTime=0.1, blendOption=Linear, blendOutTime=0.1}
    boneToModify {boneName=pack}
    translation {z=-3}
   }
   n2 = "Transform (Modify) Bone - Bone: dyn_fan_r_01"(ComponentPose=@n1, Alpha=1, Translation="0,0,0", Rotation=Get Spin result Right() -> (Spin result Right:Rotator), Scale="1,1,1") {rotationMode=BMM_Additive, rotationSpace=BCS_BoneSpace, alphaBoolBlend {blendOption=Linear}, boneToModify {boneName=dyn_fan_r_01}}
   n3 = "Transform (Modify) Bone - Bone: dyn_fan_l_01"(ComponentPose=@n2, Alpha=1, Translation="0,0,0", Rotation=Get Spin result Left() -> (Spin result Left:Rotator), Scale="1,1,1") {rotationMode=BMM_Additive, rotationSpace=BCS_BoneSpace, alphaBoolBlend {blendOption=Linear}, boneToModify {boneName=dyn_fan_l_01}}
   n4 = "Transform (Modify) Bone - Bone: dyn_motor_01"(
    ComponentPose=@n3
    Alpha=1
    Translation="0,0,0"
    Rotation=Get MotorRotate() -> (MotorRotate:Rotator)
    Scale=Get MotorScale() -> (MotorScale:Vector)
   ) {
    rotationMode=BMM_Additive
    rotationSpace=BCS_BoneSpace
    scaleMode=BMM_Additive
    alphaBoolBlend {blendOption=Linear}
    boneToModify {boneName=dyn_motor_01}
   }
   Output Pose(Result=Component To Local(ComponentPose=@n4)) {layerGroup=None}
  ]}
 EventGraph {nodes [
   n1 = Reactive() {is=title K2Node_CustomEvent}
   > Set bEnemyPlayerEliminated(bEnemyPlayerEliminated:bool=true) -> (Output_Get:bool)
   > Delay(Duration=3)
   > Set bEnemyPlayerEliminated(bEnemyPlayerEliminated=false)
   n2 = Event BlueprintUpdateAnimation()
   > Set Delta Time X(Delta Time X=@n2.DeltaTimeX)
   > n3 = Sequence()
   > Set Delta time remap(Delta time remap:double=FInterpTo(Current=Get Delta time remap() -> (Delta time remap:double), Target=1, DeltaTime=0, InterpSpeed=1)) -> (Output_Get:double)
   > Set Backpack offset(Backpack offset:bool=NOT Boolean(A=Get bIsSkyDiving())) -> (Output_Get:bool)
   n4 = "Sin (Radians)"(A=float * float(A=float * float(A=1, B=20), B=GetGameTimeInSeconds()))
   n5 = RandomFloatInRange(
    Min=float + float(A=float / float(A=@n4, B=500) {is=title K2Node_CallFunction}, B=1) {is=title K2Node_CommutativeAssociativeBinaryOperator}
    Max=1
   )
   > Set MotorScale(MotorScale:Vector=MakeVector(X=float + float(A=0, B=1), Y=@n5, Z=@n5)) -> (Output_Get:Vector)
   n6 = float * float(A=float * float(A=1, B=60), B=GetGameTimeInSeconds())
   n7 = Get bEnemyPlayerEliminated() -> (bEnemyPlayerEliminated:bool)
   n8 = RandomFloatInRange(Min=float * float(A="Sin (Radians)"(A=@n6), B=SelectFloat(A=2, B=0.6, bPickA=@n7)), Max=1)
   > Set MotorRotate(MotorRotate:Rotator=MakeRotator(Pitch=@n8, Yaw=@n8)) -> (Output_Get:Rotator)
   Branch(execute=@n3.then_1, Condition=Get ReactivityEnabled() -> (ReactivityEnabled:bool))
   > n9 = FanRotation!(execute, self, Spin:double=Get Spin() -> (Spin:double), Speed:double=Get Speed() -> (Speed:double)) -> (then, Output:double)
   > n10 = Set Spin(Spin:double=@n9) -> (Output_Get:double)
   > n11 = Set Spin result Left(Spin result Left:Rotator=MakeRotator(Roll=@n10)) -> (Output_Get:Rotator)
   > Set Spin result Right(Spin result Right:Rotator="Scale Rotator (Float)"(A:Rotator=@n11, B:float=-1) -> (ReturnValue:Rotator=~) {is=title K2Node_PromotableOperator}) -> (Output_Get:Rotator)
   Event OnCharacterCustomizationLoaded()
   > n12 = Cast To FortPlayerPawn(Object=GetOwningActor())
   > n13 = GetVariantChannelActiveIndexForItemActor!(
    execute
    Item:Athena Cosmetic Item Definition Object Reference=/BRCosmetics/Athena/Items/Cosmetics/Backpacks/Backpack_PrairieGizmo
    Actor:Actor Object Reference=@n12
    Channel:Gameplay Tag Structure=(TagName="Cosmetics.Variant.Channel.Parts")
    VariantList:Array of Gameplay Tag Structures=Make Array("[0]"=(TagName="Cosmetics.Variant.Property.Stage1"))
    bErrorIfChannelMissing:bool=false
    bErrorIfTagNotInArray:bool=false
   ) -> (then, Index:int=0, bIndexIsValid:bool, bFoundChannel:bool)
   > n14 = Branch(Condition=@n13.bIndexIsValid)
   > Set ReactivityEnabled(ReactivityEnabled:bool=true) -> (Output_Get:bool)
   > Reactive() {is=title K2Node_CallFunction}
   Set ReactivityEnabled(execute=@n14.else, ReactivityEnabled=false)
   n15 = Event BlueprintProcessAnimInputEvent()
   > Branch(Condition=IsNamed(self=@n15.AnimInputEvent, InEventName=aie.EnemyPlayerEliminated))
   > Reactive()
   Event BlueprintBeginPlay() {is=title K2Node_Event}
   > Parent: BlueprintBeginPlay() {is=title K2Node_CallParentFunction}
   > Branch(Condition=IsCosmeticObjectDisplayedInFrontendContext!(CosmeticObject:Object Reference) -> (ReturnValue:bool))
   > Delay(Duration=4)
   > Reactive()
   > Set Timer by Event(Delegate=@n1, Time=13, bLooping=true)
   TryGetPawnOwner()
  ]}
 FanRotation {nodes [
   n1 = FanRotation!() -> (then, Spin:double=~, Speed:double=~) {is=title K2Node_FunctionEntry}
   > n2 = Set Spin var(Spin var:double=@n1) -> (Output_Get:double)
   n3 = Get bEnemyPlayerEliminated()
   > Set Acceleration(Acceleration:double=FInterpTo(Current=Get Acceleration() -> (Acceleration:double), Target=SelectFloat(A=7, B=0, bPickA=@n3), DeltaTime=Get Delta Time X(), InterpSpeed=3)) -> (Output_Get:double)
   n4 = OR Boolean(A=Get bIsSkydiveFloating(), B=Get bIsParachuteOpened())
   > Set Acceleration(Acceleration=FInterpTo(Current=Get Acceleration(), Target=SelectFloat(A=1, B=0, bPickA=@n4), DeltaTime=Get Delta Time X(), InterpSpeed=2))
   n5 = Get bIsSkydiveDiveMode()
   > Set Acceleration(Acceleration=FInterpTo(Current=Get Acceleration(), Target=SelectFloat(A=3, B=0, bPickA=@n5), DeltaTime=Get Delta Time X(), InterpSpeed=2))
   > n6 = Branch(Condition="float >= float"(A=@n2, B=720))
   > n7 = Set Spin var(Spin var=float - float(A=Get Spin var() -> (Spin var:double), B=720))
   n8 = Get Velocity mult() -> (Velocity mult:double)
   n9 = Get Acceleration()
   n10 = Get Delta time remap()
   n11 = Set Spin var(
    execute=@n6.else
    Spin var=float + float!(A=Get Spin var(), B=float * float(A=@n1.Speed, B=@n10) {is=title K2Node_PromotableOperator}, C:double=float * float(A=@n9, B=@n8)) -> (ReturnValue) {is=title K2Node_PromotableOperator}
   )
   Return Node!(execute=[@n11.then, @n7.then], Output:double=@n11)
  ]}
]
variables [
 Acceleration {type=number}
 Backpack offset {type=boolean}
 bEnemyPlayerEliminated {type=boolean}
 Delta time remap {type=number}
 Delta Time X {type=number}
 MotorRotate {type=Rotator}
 MotorScale {type=Vector}
 ReactivityEnabled {type=boolean}
 Speed {type=number}
 Spin {type=number}
 Spin result Left {type=Rotator}
 Spin result Right {type=Rotator}
 Spin var {type=number}
 Variant List {type="array<GameplayTag>", default []}
 Velocity mult {type=number, default=12}
]
